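# Physics asset for the DevilRock pickaxe skeletal mesh, a two-handed melee weapon cosmetic.
# A single capsule body on the Melee_TwoHanded_Handle bone, offset mostly along the bone's Z axis, provides the weapon's whole collision shape.
# It has no constraints, so the physics representation is one body with no jointed parts.
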PhysicsAsset /BRCosmetics/Weapons/FORT_Melee/Pickaxe_DevilRock/Meshes/SK_Pickaxe_DevilRock_PhysicsAsset
bodies {Melee_TwoHanded_Handle [{shapeType=Capsule, length=72.6692, radius=21.5746, center=<1.7751 -1.3261 38.2565>, rotation=<rot -0.0257 -90 0.4669>}]}
constraints []
