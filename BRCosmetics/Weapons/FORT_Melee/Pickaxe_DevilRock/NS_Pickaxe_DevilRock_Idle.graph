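# Idle Niagara effect for the DevilRock pickaxe, producing small rising flame sprites with an orange glow and a light.
# A single FlamesNGlow emitter continuously spawns short-lived particles in a flattened, offset sphere, growing and fading them through curves.
# Emitter parameters define glow position, size and a time-of-day-scaled color, which a waveform modulates into a flickering light color.
# Two sprite renderers and a light renderer draw the result.
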NiagaraSystem /BRCosmetics/Weapons/FORT_Melee/Pickaxe_DevilRock/NS_Pickaxe_DevilRock_Idle
dependencies {
 usedDataInterfaces [&NiagaraDataInterfaceCurve, &NiagaraDataInterfaceColorCurve]
 usedDynamicInputs [&Boolean_NotOperation, &ColorFromCurve]
 usedModules [
  &SystemState
  &self:SystemSpawnScript.SystemScriptSource.SystemScriptGraph.NiagaraNodeAssignment_4.SetVariables_#1
  &self:SystemSpawnScript.SystemScriptSource.SystemScriptGraph.NiagaraNodeAssignment_1.SetVariables_#2
 ]
}
emitters {FlamesNGlow {
  emitterUpdateScript [
   &EmitterState()
   &SpawnRate(SpawnRate=5, Use Spawn Probability=false)
   /SetVariables(Emitter.PositionGlow:NiagaraPosition=<-3 0 25>) {isSetParameters=true}
   /SetVariables(
    Emitter.GlowColor:LinearColor=&MultiplyLinearColor(
     Linear Color=<rgba 1 0.1667 0 0.15>
     Scale Factor=&MakeVector4FromVectorAndFloat(
      "Vector (RGB)"=<1 1 1>
      "Float (Alpha)"=&FNTimeOfDayScalar(Night Scale=0.7, Morning Scale=0.85, Evening Scale=0.8)
     )
    )
    Emitter.GlowSize:Vector2f=<60 90>
   ) {isSetParameters=true}
   /SetVariables(Emitter.CameraOffset=15) {isSetParameters=true}
   &SpriteFacingAndAlignment(WriteAlignment=true, WriteFacing=false)
   /SetVariables(Emitter.LightColor:LinearColor=&MultiplyLinearColor(
     Linear Color=$Emitter.GlowColor
     Scale Factor=&MakeVector4FromVectorAndFloat(
      "Vector (RGB)"=<1 1.25 1>
      "Float (Alpha)"=&Waveform("Waveform [1]"=::NewEnumerator12, "[1] Frequency"=6, "[1] Amplitude Min/Max"=<12 25>)
     )
    )) {isSetParameters=true}
  ]
  particleSpawnScript [
   &InitializeParticle(
    Color=<rgba 2 0.9504 0.642 0.1>
    Lifetime=&RandomRangeFloat(Minimum=0.5, Maximum=1)
    Position Offset=<-3 0 24.1>
    Sprite Rotation Mode=::NewEnumerator2
    Sprite Size Mode=::NewEnumerator1
    Uniform Sprite Size Max=15
    Uniform Sprite Size Min=10
    UsePositionOffset=true
    Write Material Random=true
   )
   &ShapeLocation(MaskSpawnGroup=false, Non Uniform Scale=<0.2 1 0.1>, Override Seed=false, Sphere Radius=11, UseOwnerScale=true)
   &CameraOffset(Camera Offset Amount=2)
   &AddVelocity(Velocity=<0 0 20>)
   &ApplyInitialForces() {enabled=false}
  ]
  particleUpdateScript [
   &ParticleState()
   &ScaleSpriteSize(Uniform Curve Sprite Scale=DataInterface() {propertyValues {Curve {keys [
        {arriveTangent=4.2031, interpMode=RCIM_Cubic, leaveTangent=4.2031, tangentMode=RCTM_User, time=0, value=0.2}
        {interpMode=RCIM_Cubic, tangentMode=RCTM_User, time=0.33, value=1}
       ]}}})
   &ScaleColor(
    Scale Alpha=&FloatFromCurve(
     FloatCurve=DataInterface() {propertyValues {Curve {keys [
         {arriveTangent=2.0398, interpMode=RCIM_Cubic, leaveTangent=2.0398, tangentMode=RCTM_User, time=0, value=0}
         {interpMode=RCIM_Cubic, tangentMode=RCTM_User, time=0.3, value=1}
         {interpMode=RCIM_Cubic, tangentMode=RCTM_User, time=1, value=0}
        ]}}}
     Scale Curve=&RandomRangeFloat(Minimum=0.2, Maximum=0.3)
    )
    ScaleA=true
    ScaleRGB=true
   )
   &SolveForcesAndVelocity(Clamp Velocity=false, Limit Acceleration=false)
  ]
  rendererClasses [&NiagaraSpriteRendererProperties, &NiagaraLightRendererProperties]
  renderers [
   {rendererClass=&NiagaraSpriteRendererProperties}
   {rendererClass=&NiagaraSpriteRendererProperties, rendererIndex=1}
   {rendererClass=&NiagaraLightRendererProperties, rendererIndex=2}
  ]
 }}
summary {
 systemName=NS_Pickaxe_DevilRock_Idle
 emitters [{emitterName=FlamesNGlow, rendererClasses [&NiagaraSpriteRendererProperties, &NiagaraLightRendererProperties]}]
 userVariables [
  User.bisSwinging {type=NiagaraBool}
  User.FXScale {type=NiagaraFloat, defaultValue=1, description=""}
  User.RibbonIDOffset {type=NiagaraInt32}
 ]
}
