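# Niagara swing trail effect for the DevilRock pickaxe, combining fiery ribbons with drifting embers.
# TrailFlame and TrailSoft spawn anim-trail ribbons between two weapon sockets, fading through color and alpha curves, with TrailSoft adding dynamic material parameters.
# Embers spawns by distance and rate at a sampled socket, inheriting its velocity under curl noise and drag.
# Particle partitions vary ember motion and set visibility tags across three sprite renderers.
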NiagaraSystem /BRCosmetics/Weapons/FORT_Melee/Pickaxe_DevilRock/NS_Pickaxe_DevilRock_Trail
dependencies {
 usedDataInterfaces [
  &NiagaraDataInterfaceColorCurve
  &NiagaraDataInterfaceCurve
  &NiagaraDataInterfaceSpriteRendererInfo
  &NiagaraDataInterfaceSocketReader
 ]
 usedDynamicInputs []
 usedModules [
  &self:SystemSpawnScript.SystemScriptSource.SystemScriptGraph.NiagaraNodeAssignment_0.SetVariables_#1
  &SystemState
  &self:SystemSpawnScript.SystemScriptSource.SystemScriptGraph.NiagaraNodeAssignment_1.SetVariables_#2
 ]
}
emitters {
 Embers {
  emitterUpdateScript [
   &EmitterState(Life Cycle Mode=::NewEnumerator1, Loop Behavior=::NewEnumerator1, Loop Delay=0.07, Loop Duration=0.12, UseLoopDelay=true)
   &SpawnPerUnit(Module.Use Max Movement Threshold=true, Module.Use Movement Tolerance=false, Spawn Spacing=30, Use Spawn Probability=false)
   &SpawnRate(SpawnRate=100, Use Spawn Probability=true)
  ]
  particleSpawnScript [
   &PartitionParticles()
   &InitializeParticle(
    Color=<rgba 4 2.4696 2.4696 1>
    Lifetime=&MergeHelper_FloatSwitch(
     0=&RandomRangeFloat(Minimum=0.3, Maximum=0.5)
     1=&RandomRangeFloat(Minimum=0.4, Maximum=0.5)
     2=?
     3=?
     Index=$Particles.PartitionParticles.Partition
     Num=::NewEnumerator0
    )
    Sprite Rotation Mode=::NewEnumerator2
    Sprite Size Max=<1.5 4>
    Sprite Size Min=<0.7 2>
    Sprite Size Mode=::NewEnumerator2
    UsePositionOffset=false
    Write Material Random=true
   )
   &SocketLocation(
    Interpolate Socket Sampling=true
    SamplePosOffset=true
    Sampled Position Offset=<0 0 20>
    Socket Index=2
    Socket Reader=$System.Socket Reader
    Socket Sampling Mode=::NewEnumerator3
    VelocitySpeedLimit=false
   )
   &ShapeLocation(
    MaskSpawnGroup=false
    Offset Mode=::NewEnumerator0
    Override Seed=false
    Rotation Axis=<0 1 0>
    Rotation Mode=::NewEnumerator1
    Sphere Radius=&MergeHelper_FloatSwitch(0=8, 1=12, 2=?, 3=?, Index=$Particles.PartitionParticles.Partition, Num=::NewEnumerator0)
    Sphere Surface Distribution=0.7976
    UseOwnerScale=true
   )
   &AddVelocity(
    Module.Use Falloff Distance=false
    Override Seed=false
    Velocity Mode=::NewEnumerator1
    Velocity Origin=$Output.SocketLocation.SampledPosition
    Velocity Speed=&MergeHelper_FloatSwitch(0=20, 1=1, 2=?, 3=?, Index=$Particles.PartitionParticles.Partition, Num=::NewEnumerator0)
   )
   &InheritVelocity(
    Inherited Velocity Speed Limit:NiagaraFloat=&RandomRangeFloat(Minimum=100, Maximum=200)
    Speed Limit=true
    UseSpeedThreshold=false
    Velocity Source=$Output.SocketLocation.SampledVelocity
   )
   /SetVariables(Particles.VisibilityTag=$Particles.PartitionParticles.Partition) {isSetParameters=true}
  ]
  particleUpdateScript [
   &ParticleState()
   &ScaleSpriteSize(Uniform Curve Sprite Scale=DataInterface() {propertyValues {Curve {keys [
        <at 0 0 RCIM_Cubic>
        <at 0.2 1 RCIM_Cubic>
        {arriveTangent=-2.2922, interpMode=RCIM_Cubic, leaveTangent=-2.2922, tangentMode=RCTM_User, time=1, value=0}
       ]}}})
   &Drag(
    Drag=&MergeHelper_FloatSwitch(0=3, 1=8, 2=?, 3=?, Index=$Particles.PartitionParticles.Partition, Num=::NewEnumerator0)
    Use Linear Drag=true
    Use Rotational Drag=true
   )
   &ScaleColor(
    Scale Alpha=&FloatFromCurve(FloatCurve=DataInterface() {propertyValues {Curve {keys [
         {interpMode=RCIM_Cubic, tangentMode=RCTM_User, time=0.1, value=0}
         {arriveTangent=-0.625, interpMode=RCIM_Cubic, leaveTangent=-0.625, tangentMode=RCTM_User, time=0.2, value=1}
        ]}}})
    Scale RGB=&MakeVectorFromLinearColorRGB(Color=&ColorFromCurve(ColorCurve=DataInterface() {propertyValues {
        AlphaCurve {keys [<at 0 1 RCIM_Cubic>]}
        BlueCurve {keys [<at 0 0.18 RCIM_Cubic>,<at 0.2838 0.03>,<at 0.922 0>]}
        GreenCurve {keys [<at 0 1.635 RCIM_Cubic>,<at 0.2838 0.4755>,<at 0.922 0.0033>]}
        RedCurve {keys [<at 0 6 RCIM_Cubic>,<at 0.2838 3>,<at 0.922 0.2>]}
       }}))
    ScaleA=true
    ScaleRGB=true
   )
   /SetVariables(Particles.Glow03_Size=&Multiply_Vector2D_ByFloat(
     Float=&FloatFromCurve(
      FloatCurve=DataInterface() {propertyValues {Curve {keys [
          <at 0 0 RCIM_Cubic>
          <at 0.2 1.3 RCIM_Cubic>
          {arriveTangent=-0.1423, interpMode=RCIM_Cubic, leaveTangent=-0.1423, tangentMode=RCTM_User, time=1, value=0}
         ]}}}
      Scale Curve=10
     )
     Vector2D=<0.7 1.5>
    )) {isSetParameters=true}
   /SetVariables(Particles.Glow03_Color=&MakeLinearColorFromVectorAndFloat(
     "Vector (RGB)"=&MakeVectorFromLinearColorRGB(Color=&MultiplyLinearColor(Linear Color=<rgba 1 0.1333 0 1>, Scale Factor=&MakeVector4_FromFloat(Float=1)))
     "Float (Alpha)"=&Multiply_Float(A=0.2, B=1)
    )) {isSetParameters=true}
   &GravityForce(Gravity=&Multiply_VectorByFloat(
     Vector=&RandomRangeVector(Minimum=<0 0 50>, Maximum=<0 0 100>)
     Float=&MergeHelper_FloatSwitch(0=1, 1=-2, 2=?, 3=?, Index=$Particles.PartitionParticles.Partition, Num=::NewEnumerator0)
    ))
   &CurlNoiseForce(
    Noise Frequency=&RandomRangeFloat(Minimum=60, Maximum=100)
    Noise Strength=&RandomRangeFloat(Minimum=400, Maximum=500)
    Pan Noise=true
    Pan Noise Field=<0.2 0 0.3>
   )
   &SolveForcesAndVelocity(Clamp Velocity=false, Limit Acceleration=false)
   /SetVariables(
    Particles.AshesSpriteSize:Vector2f=&Vector2DFromFloat(Value=&Multiply_Float(
      A=&RandomRangeFloat(Minimum=4, Maximum=10)
      B=&FloatFromCurve(FloatCurve=DataInterface() {propertyValues {Curve {keys [<at 0 1>,<at 1 0>]}}})
     ))
    Particles.ColorAshes:LinearColor=&MultiplyLinearColor(
     Linear Color=&ColorFromCurve(ColorCurve=DataInterface() {propertyValues {
        AlphaCurve {keys [<at 0 1 RCIM_Cubic>]}
        BlueCurve {keys [<at 0.134 60 RCIM_Cubic>,<at 0.7925 1 RCIM_Cubic>]}
        GreenCurve {keys [<at 0.134 108 RCIM_Cubic>,<at 0.7925 1 RCIM_Cubic>]}
        RedCurve {keys [<at 0.134 300 RCIM_Cubic>,<at 0.7925 1 RCIM_Cubic>]}
       }})
     Scale Factor=&MakeVector4_FromLinearColor(LinearColor=&RandomRangeLinearColor(Minimum=<rgba 1 0.1466 0.0788 1>, Maximum=<rgba 1 0.427 0.1134 1>))
    )
   ) {isSetParameters=true}
   &SubUVAnimation(
    "Sprite Renderer "=DataInterface() {propertyValues {SpriteRenderer=&self:Embers_1.NiagaraSpriteRendererProperties_1}}
    Start Frame Offset:NiagaraInt32=&RandomRangeInt(Minimum=0, Maximum=12)
    UseEndFrame=false
    UseStartFrame=false
   ) {enabled=false}
  ]
  rendererClasses [&NiagaraSpriteRendererProperties]
  renderers [
   {rendererClass=&NiagaraSpriteRendererProperties}
   {rendererClass=&NiagaraSpriteRendererProperties, rendererIndex=1}
   {rendererClass=&NiagaraSpriteRendererProperties, rendererIndex=2}
  ]
 }
 TrailFlame {
  emitterSpawnScript [&FN_RibbonIDOffsetUtility()]
  emitterUpdateScript [
   &EmitterState(Life Cycle Mode=::NewEnumerator1, Loop Behavior=::NewEnumerator1, Loop Delay=0.03, Loop Duration=0.25, UseLoopDelay=true)
   &FN_SpawnAnimTrail(AnimTrailWidthMode=::FromSecond)
  ]
  particleSpawnScript [
   &FN_InitializeAnimTrailBoneIndecies(First Bone or Socket Index=1, Second Bone or Socket Index=2)
   &FN_InitializeAnimTrail(AnimTrailWidth=2.4, Lifetime=0.21)
   &AddVelocity(Velocity=<1000 0 0>) {enabled=false}
  ]
  particleUpdateScript [
   &SolveForcesAndVelocity(Clamp Velocity=false, Limit Acceleration=false)
   &FN_UpdateAnimTrail() {module=FN_UpdateAnimTrail001}
   &ScaleColor(
    Scale Alpha=&Multiply_Float(
     A=&FloatFromCurve(FloatCurve=DataInterface() {propertyValues {Curve {keys [
          {arriveTangent=2.9017, interpMode=RCIM_Cubic, leaveTangent=2.9017, tangentMode=RCTM_User, time=0, value=0}
          {arriveTangent=-0.0116, interpMode=RCIM_Cubic, leaveTangent=-0.2338, tangentMode=RCTM_Break, time=0.15, value=1}
          {arriveTangent=-4.5481, interpMode=RCIM_Cubic, tangentMode=RCTM_Break, time=0.8, value=0}
         ]}}})
     B=$System.SystemFade
    )
    Scale RGB=&MakeVectorFromLinearColorRGB(Color=&ColorFromCurve(ColorCurve=DataInterface() {propertyValues {
        AlphaCurve {keys []}
        BlueCurve {keys [<at 0.0507 0.9>,<at 0.4755 0.07>,<at 0.8009 0.04>]}
        GreenCurve {keys [<at 0.0507 3>,<at 0.4755 0.256>,<at 0.8009 0.076>]}
        RedCurve {keys [<at 0.0507 4.5>,<at 0.4755 1>,<at 0.8009 0.4>]}
       }}))
    ScaleA=true
    ScaleRGB=true
   )
   &ScaleRibbonWidth(Ribbon Width Scale=&FloatFromCurve(FloatCurve=DataInterface() {propertyValues {Curve {keys [
         {arriveTangent=-1.1425, interpMode=RCIM_Cubic, leaveTangent=-1.1425, tangentMode=RCTM_User, time=0, value=1}
         {arriveTangent=-0.0978, interpMode=RCIM_Cubic, leaveTangent=-0.0978, tangentMode=RCTM_User, time=1, value=0.5}
        ]}}}))
  ]
  rendererClasses [&NiagaraRibbonRendererProperties]
  renderers [{rendererClass=&NiagaraRibbonRendererProperties}]
 }
 TrailSoft {
  emitterSpawnScript [&FN_RibbonIDOffsetUtility()]
  emitterUpdateScript [
   &EmitterState(Life Cycle Mode=::NewEnumerator1, Loop Behavior=::NewEnumerator1, Loop Delay=0.03, Loop Duration=0.25, UseLoopDelay=true)
   &FN_SpawnAnimTrail(AnimTrailWidthMode=::FromSecond)
  ]
  particleSpawnScript [
   &FN_InitializeAnimTrailBoneIndecies(First Bone or Socket Index=1, Second Bone or Socket Index=2)
   &FN_InitializeAnimTrail(AnimTrailWidth=5, Color=<rgba 7 4.55 2.8 0.2>, Lifetime=0.22)
   &AddVelocity(Velocity=<1000 0 0>) {enabled=false}
  ]
  particleUpdateScript [
   &SolveForcesAndVelocity(Clamp Velocity=false, Limit Acceleration=false)
   &FN_UpdateAnimTrail() {module=FN_UpdateAnimTrail001}
   &ScaleColor(
    Scale Alpha=&Multiply_Float(
     A=&FloatFromCurve(
      FloatCurve=DataInterface() {propertyValues {Curve {keys [
          {arriveTangent=-0.0116, interpMode=RCIM_Cubic, leaveTangent=-1.309, tangentMode=RCTM_Break, time=0, value=1}
          {arriveTangent=-0.8124, interpMode=RCIM_Cubic, tangentMode=RCTM_Break, time=0.6, value=0}
         ]}}}
      Scale Curve=0.75
     )
     B=$System.SystemFade
    )
    Scale RGB=&MakeVectorFromLinearColorRGB(Color=&ColorFromCurve(ColorCurve=DataInterface() {propertyValues {
        AlphaCurve {keys []}
        BlueCurve {keys [<at 0.0436 1.25>,<at 0.6889 0.015>]}
        GreenCurve {keys [<at 0.0436 1.875>,<at 0.6889 0.1201>]}
        RedCurve {keys [<at 0.0436 2.5>,<at 0.6889 0.5>]}
       }}))
    ScaleA=true
    ScaleRGB=true
   )
   &ScaleRibbonWidth(Ribbon Width Scale=&FloatFromCurve(FloatCurve=DataInterface() {propertyValues {Curve {keys [
         {arriveTangent=-1.3947, interpMode=RCIM_Cubic, leaveTangent=-1.3947, tangentMode=RCTM_User, time=0, value=1}
         {arriveTangent=-0.0623, interpMode=RCIM_Cubic, leaveTangent=-0.0623, tangentMode=RCTM_User, time=1, value=0.2}
        ]}}})) {enabled=false}
   &DynamicMaterialParameters(
    Index 0 Param 3=&FloatFromCurve(FloatCurve=DataInterface() {propertyValues {Curve {keys [
         {arriveTangent=1.0387, interpMode=RCIM_Cubic, leaveTangent=1.0387, tangentMode=RCTM_User, time=0, value=0}
         {arriveTangent=0.3495, interpMode=RCIM_Cubic, leaveTangent=0.3495, tangentMode=RCTM_User, time=1, value=1}
        ]}}})
    Index 1 Param 1=&FloatFromCurve(FloatCurve=DataInterface() {propertyValues {Curve {keys [
         {arriveTangent=0.6953, interpMode=RCIM_Cubic, leaveTangent=0.6953, tangentMode=RCTM_User, time=0.3, value=0}
         {arriveTangent=1.3769, interpMode=RCIM_Cubic, leaveTangent=1.3769, tangentMode=RCTM_User, time=1, value=1}
        ]}}})
    Module.Index0Float0Write=false
    Module.Index0Float1Write=false
    Module.Index0Float2Write=true
    Module.Index0Float3Write=false
    Module.Index1Float0Write=false
    Module.Index1Float1Write=false
    Module.Index1Float2Write=false
    Module.Index1Float3Write=false
    Write Parameter Index 1=::NewEnumerator1
   )
   /SetVariables(Particles.ColorSharp:LinearColor=&MultiplyLinearColor(
     Linear Color=<rgba 1 1 1 1>
     Scale Factor=&MakeVector4FromVectorAndFloat(
      "Vector (RGB)"=<0.4 1.5 0.6>
      "Float (Alpha)"=&FloatFromCurve(
       FloatCurve=DataInterface() {propertyValues {Curve {keys [
           {arriveTangent=-1.9862, interpMode=RCIM_Cubic, leaveTangent=-1.9862, tangentMode=RCTM_User, time=0, value=1}
           {arriveTangent=-0.4328, interpMode=RCIM_Cubic, leaveTangent=-0.4328, tangentMode=RCTM_User, time=0.4, value=0}
          ]}}}
       Scale Curve=0.9
      )
     )
    )) {isSetParameters=true}
  ]
  rendererClasses [&NiagaraRibbonRendererProperties]
  renderers [
   {rendererClass=&NiagaraRibbonRendererProperties}
   {rendererClass=&NiagaraRibbonRendererProperties, rendererIndex=1}
  ]
 }
}
summary {
 systemName=NS_Pickaxe_DevilRock_Trail
 emitters [
  {emitterName=TrailSoft, rendererClasses [&NiagaraRibbonRendererProperties]}
  {emitterName=Embers, rendererClasses [&NiagaraSpriteRendererProperties]}
  {emitterName=TrailFlame, rendererClasses [&NiagaraRibbonRendererProperties]}
 ]
 userVariables [
  User.bIsFrontEnd {type=NiagaraBool}
  User.bIsFrontEndPreview {type=NiagaraBool}
  User.RibbonIDOffset {type=NiagaraInt32}
 ]
}
